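# Animation Blueprint for the PrecisionMongoose backpack that nudges its pack bone to follow the wearer's movement and pose.
# A chain of additive Modify Bone nodes layers offsets, mostly blended in and out by bools for melee, high upward velocity, grounded, skydive floating and glider states.
# The update event sets those bools and derives rotations from the owner mesh's spine twist deltas and left clavicle yaw.
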AnimBlueprint /BRCosmetics/Accessories/FORT_Backpacks/Backpack_M_MED_PrecisionMongoose/Meshes/M_MED_PrecisionMongoose_Pack_AnimBP
target_skeleton=&/M_MED_PrecisionMongoose_Pack_Skeleton
graphs [
 AnimGraph {nodes [
   n1 = "Transform (Modify) Bone - Bone: pack"!(
    ComponentPose=Local To Component(LocalPose=Input Pose())
    bAlphaBoolEnabled:bool=Get bIsPlayingMeleeAnim()
    Translation="0,0,0"
    Rotation=Get rawRotationSpine() -> (rawRotationSpine:Rotator)
    Scale
   ) -> (Pose) {
    alphaInputType=Bool
    rotationMode=BMM_Additive
    translationMode=BMM_Additive
    alphaBoolBlend {blendInTime=0.1, blendOption=Linear, blendOutTime=0.16}
    boneToModify {boneName=pack}
   }
   n2 = "Transform (Modify) Bone - Bone: pack"(
    ComponentPose=@n1
    bAlphaBoolEnabled=Get bIsPlayingMeleeAnim()
    Translation="0,0,0"
    Rotation=Get Return Value Rotation() -> (Return Value Rotation:Rotator)
   ) {
    alphaInputType=Bool
    rotationMode=BMM_Additive
    translationMode=BMM_Additive
    alphaBoolBlend {blendInTime=0.1, blendOption=Linear, blendOutTime=0.16}
    boneToModify {boneName=pack}
   }
   n3 = "Transform (Modify) Bone - Bone: pack"(ComponentPose=@n2, bAlphaBoolEnabled=Get bGreaterVel() -> (bGreaterVel:bool), Translation="0,-3,0", Rotation="0,0,10") {
    alphaInputType=Bool
    rotationMode=BMM_Additive
    translationMode=BMM_Additive
    alphaBoolBlend {blendInTime=0.1, blendOption=Linear, blendOutTime=0.5}
    alphaScaleBiasClamp {bInterpResult=true, interpSpeedDecreasing=0.5, interpSpeedIncreasing=5}
    boneToModify {boneName=pack}
    rotation {roll=10}
    translation {y=-3}
   }
   n4 = "Transform (Modify) Bone - Bone: pack"(ComponentPose=@n3, bAlphaBoolEnabled=Get NewVar() -> (NewVar:bool), Translation="0,1,0", Rotation="0,0,-2") {
    alphaInputType=Bool
    rotationMode=BMM_Additive
    translationMode=BMM_Additive
    alphaBoolBlend {blendInTime=0.1, blendOption=Linear, blendOutTime=0.1}
    alphaScaleBiasClamp {bInterpResult=true, interpSpeedDecreasing=0.5, interpSpeedIncreasing=5}
    boneToModify {boneName=pack}
    rotation {roll=-2}
    translation {y=1}
   }
   n5 = "Transform (Modify) Bone - Bone: pack"!(
    ComponentPose=@n4
    bAlphaBoolEnabled="Get Fly_condition "() -> ("Fly_condition ":bool)
    Translation="0,-1,0"
    Rotation_Roll:float=Get!() -> (RotateSpineCond_inSky_Roll:float, RotateSpineCond_inSky_Pitch:float, RotateSpineCond_inSky_Yaw:float).RotateSpineCond_inSky_Yaw
    Rotation_Pitch:float=0
    Rotation_Yaw:float=0
    Scale
   ) -> (Pose) {
    alphaInputType=Bool
    rotationMode=BMM_Additive
    translationMode=BMM_Additive
    alphaBoolBlend {blendInTime=1, blendOption=Linear, blendOutTime=0.4}
    alphaScaleBiasClamp {bInterpResult=true, interpSpeedDecreasing=0.5, interpSpeedIncreasing=5}
    boneToModify {boneName=pack}
    rotation {roll=10}
    translation {y=-1}
   }
   n6 = "Transform (Modify) Bone - Bone: pack"!(ComponentPose=@n5, bAlphaBoolEnabled, Translation="0,-1,0", Rotation:Rotator="0,0,5", Scale) -> (Pose) {
    alphaInputType=Bool
    rotationMode=BMM_Additive
    translationMode=BMM_Additive
    alphaBoolBlend {blendInTime=0.3, blendOption=Linear, blendOutTime=0.3}
    alphaScaleBiasClamp {bInterpResult=true, interpSpeedDecreasing=0.5, interpSpeedIncreasing=5}
    boneToModify {boneName=pack}
    rotation {roll=5}
    translation {y=-1}
   }
   n7 = "Transform (Modify) Bone - Bone: pack"!(
    ComponentPose=@n6
    bAlphaBoolEnabled=Get Para Glide Open Glider() -> (Para Glide Open Glider:bool)
    Translation="0,-1,0"
    Rotation_Roll:float=Get!() -> (RotateSpineCond_inSky_max_Roll:float, RotateSpineCond_inSky_max_Pitch:float, RotateSpineCond_inSky_max_Yaw:float).RotateSpineCond_inSky_max_Yaw
    Rotation_Pitch:float=0
    Rotation_Yaw:float=0
    Scale
   ) -> (Pose) {
    alphaInputType=Bool
    rotationMode=BMM_Additive
    alphaBoolBlend {blendInTime=0.3, blendOption=Linear, blendOutTime=5}
    boneToModify {boneName=pack}
    translation {y=-1}
   }
   Output Pose(Result=Component To Local(ComponentPose=@n7)) {layerGroup=None}
  ]}
 EventGraph {nodes [
   Event BlueprintUpdateAnimation()
   > n1 = Sequence!(execute) -> (then_0, then_1, then_2:Exec, then_3:Exec, then_4:Exec)
   n2 = Sequence!(execute=@n1.then_2) -> (then_0, then_1, then_2)
   n3 = "Min (Float)"(
    A:double=GetAlphaFromReferenceTransformDelta(SkeletalMeshComponent=GetOwnerSkeletalMeshComponent(), BoneName=spine_04, BaseName=spine_03, InputRange=(X=0,Y=1), TwistAxis="0,0,1").RawOutput
    B:double=0
   ) -> (ReturnValue:double) {is=title K2Node_CommutativeAssociativeBinaryOperator}
   n4 = Invert Rotator!(A:Rotator=MakeRotator(Yaw=@n3)) -> (ReturnValue:Rotator)
   n5 = BreakRotator(InRot=@n4)
   > Set RotateSpineCond_inSky(RotateSpineCond_inSky:Rotator=MakeRotator(Yaw=float * float(A=@n5.Yaw, B=4) {is=title K2Node_PromotableOperator})) -> (Output_Get:Rotator)
   > "Set Fly_condition "("Fly_condition ":bool=AND Boolean(A=Get bIsSkydiveFloating(), B="float < float"(A=@n3, B=-15))) -> (Output_Get:bool)
   Set Skel_Mesh(execute=@n2.then_2, Skel_Mesh=GetOwnerSkeletalMeshComponent())
   > Set Para Glide Open Glider(Para Glide Open Glider:bool=Get bParaGlide_OpenGlider!(self:Fort Player Anim Instance Object Reference=GetOwnerFortPlayerAnimInstance!(self:Custom Character Part Anim Instance Object Reference) -> (ReturnValue:Fort Player Anim Instance Object Reference)) -> (bParaGlide_OpenGlider:bool)) -> (Output_Get:bool)
   n6 = BreakRotator(InRot=@n4)
   Set RotateSpineCond_inSky_max(execute=@n2.then_1, RotateSpineCond_inSky_max:Rotator=MakeRotator(Yaw=float * float(A=@n6.Yaw, B=5))) -> (Output_Get:Rotator)
   n7 = GetOwnerSkeletalMeshComponent()
   n8 = "Min (Float)"(
    A=BreakRotator(InRot=GetSocketTransform!(self=@n7, InSocketName=clavicle_l, TransformSpace=RTS_Component) -> (ReturnValue_Location:Vector, ReturnValue_Rotation:Rotator, ReturnValue_Scale:Vector).ReturnValue_Rotation).Yaw
    B=0
   )
   Set Return Value Rotation(execute=@n1.then_4, Return Value Rotation:Rotator=MakeRotator(Yaw=float / float(A=@n8, B=4))) -> (Output_Get:Rotator)
   n9 = Get bIsPlayingMeleeAnim()
   n10 = AND Boolean(A=AND Boolean(A=Get bIsOnGround() -> (bIsOnGround:bool), B=NOT Boolean(A=@n9)), B=NOT Boolean(A=Get bIsJumping()))
   Set NewVar(execute=@n1.then_3, NewVar:bool=@n10) -> (Output_Get:bool)
   n11 = "Max (Float)"(A=Get PawnVelocityZ(), B=0)
   n12 = Get bIsParachuteOpened()
   n13 = Branch(execute=@n1.then_0, Condition=AND Boolean(A="float > float"(A=@n11, B=3000), B="Not Equal (Boolean)"(A=@n12, B=true) -> (ReturnValue=~)))
   > Set bGreaterVel(bGreaterVel:bool=true) -> (Output_Get:bool)
   Set bGreaterVel(execute=@n13.else, bGreaterVel=false)
   n14 = Branch(execute=@n1.then_1, Condition=Get bIsParachuteOpened())
   > Set RawRotateSpine_InSky(RawRotateSpine_InSky:float=0) -> (Output_Get:float)
   n15 = Branch(execute=@n14.else, Condition=Get bIsSkyDiving())
   > Set RawRotateSpine_InSky(RawRotateSpine_InSky=1)
   Branch(execute=@n15.else, Condition=Get bIsPlayingMeleeAnim())
   n16 = "Not Equal (Boolean)"(A=Get bIsSkydiveFloating(), B=true) -> (ReturnValue=~)
   n17 = "Not Equal (Boolean)"(A=Get bIsSkydiveDiveMode(), B=true) -> (ReturnValue=~)
   n18 = GetAlphaFromReferenceTransformDelta(SkeletalMeshComponent=GetOwnerSkeletalMeshComponent(), BoneName=spine_05, BaseName=spine_04, InputRange=(X=0,Y=1), TwistAxis="0,0,1")
   > Branch(Condition=AND Boolean(A="float < float"(A=@n18.RawOutput, B=-3), B=AND Boolean(A=@n17, B=@n16)))
   n19 = float - float(B=MakeRotator!(Roll, Pitch, Yaw=@n18.RawOutput) -> (ReturnValue_Roll:float, ReturnValue_Pitch:float, ReturnValue_Yaw:float).ReturnValue_Yaw)
   > Set rawRotationSpine(rawRotationSpine:Rotator=MakeRotator!(Roll=float / float(A=@n19, B=2), Pitch, Yaw) -> (ReturnValue:Rotator)) -> (Output_Get:Rotator)
   AND Boolean(A=@n10, B=NOT Boolean(A=Get bIsCrouching()))
  ]}
]
variables [
 bGreaterVel {type=boolean}
 "Fly_condition " {type=boolean}
 NewVar {type=boolean}
 Para Glide Open Glider {type=boolean}
 RawRotateSpine_InSky {type=number}
 rawRotationSpine {type=Rotator}
 Return Value Rotation {type=Rotator}
 RotateSpineCond_inSky {type=Rotator}
 RotateSpineCond_inSky_max {type=Rotator}
 Skel_Mesh {type=SkeletalMeshComponent}
]
